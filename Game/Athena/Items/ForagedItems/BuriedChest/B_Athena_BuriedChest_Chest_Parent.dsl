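; Parent Blueprint for a buried treasure chest that can be interacted with or destroyed only once it is dug up.
; Rep-notifies apply replicated location and rotation, and scale ambient chime volume with the dig count.
; If the chest is unsearched, each dig update also starts the sound's visual indicator and, after enough digs, reveals the treasure light.
; The construction script adds an actor tag so placed chests receive the correct chrome material.
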(fn BlueprintCanInteract
  (params (InteractingPawn :type Fort%20Pawn%20Object%20Reference) (InteractionBeingAttempted :type EInteractionBeingAttempted%20Enum :default "FirstInteraction") (InteractionType :type TInteractionType%20Enum :default "IT_NoInteraction"))
  (return (Variables|Default|GetDugUp)))

(fn OnRep_ChestCurrentLoc
  (Transformation|SetActorLocation (Variables|Default|GetChestCurrentLoc)))

(fn OnRep_ChestCurrentRot
  (Transformation|SetActorRotation (Variables|Default|GetChestCurrentRot)))

(fn OnRep_TimesDug
  ;[1 1 1] When the Chest is dug up some slowly increase multiplier volume
  (if (not (Variables|Container|GetbAlreadySearched))
    (Components|Activation|Activate (Variables|Default|GetChest_Ambient_Sound) false)
    (CallFunction|StartAudiovisualIndicator (Variables|Default|GetChest_Ambient_Sound) "Chest" (Variables|Default|GetChestChimeVisualUpdate)) ; Turn on the audio visuals
    (Variables|Default|SetEnableAudioIndicator true)
    (if (>= (Variables|Default|GetTimesDug) 6)
      (Development|SetHiddeninGame (Variables|Default|GetTreasureLight) false false)
      (else
        _)))
  (Audio|Components|Audio|SetVolumeMultiplier (Variables|Default|GetChest_Ambient_Sound) (* (Variables|Default|GetTimesDug) 0.2) :!class /Script/Engine.AudioComponent))

(fn ShouldDie
  (params (Damage :type Float%20%28single-precision%29) (EventInstigator :type Controller%20Object%20Reference) (DamageCauser :type Actor%20Object%20Reference) (bIsTest :type Boolean) (EffectContext :type Gameplay%20Effect%20Context%20Handle%20Structure%20%28by%20ref%29))
  (bind _returnvalue (|CallParentFunction Damage EventInstigator DamageCauser bIsTest EffectContext))
  (return (and _returnvalue (Variables|Default|GetDugUp))))

(fn ConstructionScript
  ;[0.32 0.9402 1] Make sure actors that are already placed in levels get this tag so that they get the right chrome material added to them.
  (|CallParentFunction)
  (Utilities|Array|AddUnique (Variables|Actor|GetTags) "Lackluster.Material.TieredChestAthena"))
