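; Buried prepper chest blueprint, a tiered-chest child that shuts off its lure effects and plays an open sound when its openable cosmetics are hidden.
; The construction script registers its custom aura component as the chest's custom Niagara effects.
; On hide, a DoOnce hides the smoke sheet, destroys the aura, turns off the light, fades ambient audio, and clears audiovisual indicators, wiggle materials, mobile selection cues and custom depth.
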(event EventBeginPlay
  (|CallParentFunction) ; This node is disabled and will not be called.\nDrag off pins to build functionality.
)

(event Collision|EventActorBeginOverlap
  (params (OtherActor :type Actor%20Object%20Reference)))

(event EventTick
  (params (DeltaSeconds :type Float%20%28single-precision%29)))

(event EventHideChestOpenableCosmetics
  (Utilities|FlowControl|DoOnce false
    (:Completed
      (Rendering|SetVisibility (Variables|Default|GetCustom%20Smoke%20Sheet) false false)
      (Components|DestroyComponent (Variables|Default|GetCustom%20Chest%20Aura))
      (CallFunction|SetLightVisibility false)
      (Audio|Components|Audio|FadeOut (Variables|Default|GetChest_Ambient_Sound) 0.5 0.0 "Linear" :!class /Script/Engine.AudioComponent)
      (CallFunction|StopAudiovisualIndicator)
      (CallFunction|CleanupWiggleMIDs)
      (Components|Timeline|Stop (Variables|Tiered_Chest_Athena_Parent|GetMobileSelectedTL))
      (CallFunction|SetVisibleMobileInteractIcon false)
      (Rendering|SetRenderCustomDepth (Variables|Debug|GetStaticMeshComponent) false)
      (Audio|PlaySoundatLocation (Variables|Sound|GetOpen%20Chest%20Sound) (Transformation|GetActorLocation)))))

(fn ConstructionScript
  (|CallParentFunction)
  (Variables|Default|SetCustom%20Niagara%20Effects (Utilities|Array|MakeArray (Variables|Default|GetCustom%20Chest%20Aura))))
