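; Parent Blueprint for the Battle Royale buried-chest prop, which players dig up with harvesting tools.
; Once playlist data is ready, the server spawns a collision-disabled chest and traces the ground normal.
; Each harvest hit plays FX, raises and tilts the chest along a rotation pattern, and grows the dirt.
; At max hits, MaxDug's RepNotify unbinds damage, clears hitbox collision, and makes the chest collidable and damageable.
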(event Custom|OnBeenDug
  (Variables|Default|SetBeenDug true))

(event EventBeginPlay
  ;[1 1 1] Spawn In the Chest
  (Class|FortAsyncActionCurrentPlaylistReady|PlaylistDataReadyAsync
    (:OnReady
      (if (Networking|IsServer)
        (Collision|SetCollisionEnabled (Variables|Debug|GetStaticMeshComponent) "NoCollision")
        (if (Networking|IsServer)
          (CallFunction|FindPlacementNormal)
          (bind _returnvalue (Game|SpawnActorfromClass (Variables|Default|GetChestClass) (Math|Transform|MakeTransform (+ (Transformation|GetActorLocation) (Variables|Default|GetZOffset)) (Transformation|GetActorRotation)) "AlwaysSpawn" "SelectDefaultAtRuntime" (Variables|Actor|GetInstigator) (Variables|Actor|GetInstigator)))
          (bind _output_get (Variables|Default|SetChestObjectPure _returnvalue))
          (bind _asb_athena_buried_chest_chest_parent (Utilities|Casting|CastToB_Athena_BuriedChest_Chest_Parent _output_get))
          (bind _output_get_2 (Variables|Default|SetChestObject _asb_athena_buried_chest_chest_parent))
          (Variables|Default|SetChestInitialLoc (Transformation|GetActorLocation _output_get_2))
          (Utilities|IsValid (Variables|/Game/Building/ActorBlueprints/Containers/Tiered_Chest_Athena_Parent.Tiered_Chest_Athena_Parent_C|GetItemMesh _output_get_2)
            (:Is%20Valid
              (Collision|SetCollisionEnabled (Variables|/Game/Building/ActorBlueprints/Containers/Tiered_Chest_Athena_Parent.Tiered_Chest_Athena_Parent_C|GetItemMesh _output_get_2) "NoCollision"))
            (:Is%20Not%20Valid
              _))
          (Utilities|IsValid (Variables|/Script/FortniteGame.BuildingSMActor|GetStaticMeshComponent _output_get_2)
            (:Is%20Valid
              (Collision|SetCollisionEnabled (Variables|/Script/FortniteGame.BuildingSMActor|GetStaticMeshComponent _output_get_2) "NoCollision"))
            (:Is%20Not%20Valid
              _))
          (Default|CallOnChestSpawned)
          ;[1 1 1] Bind Damage and Call Replecated Event
          (EventDispatchers|BindEventtoOnDamaged (delegate Custom|OnDamageTaken))
          (else
            (Development|PrintString "Client Booty Prop: I AM HERE!")))))))

(event Custom|OnDamageTaken
  (params (DamagedActor :type Actor%20Object%20Reference) (Damage :type Float%20%28single-precision%29) (InstigatedBy :type Controller%20Object%20Reference) (DamageCauser :type Actor%20Object%20Reference) (HitLocation :type Vector) (FHitComponent :type Primitive%20Component%20Object%20Reference) (BoneName :type Name) (Momentum :type Vector))
  ;[1 1 1] Bind Damage and Call Replecated Event
  (bind _asfort_weapon (Utilities|Casting|CastToFortWeapon DamageCauser))
  (if (Utilities|Enum|Equal(Enum) (Item|GetWeaponDataItemType _asfort_weapon) "WeaponHarvest")
    (Math|Integer|IncrementInt (Variables|Default|GetNumTimesHit)
      (:%20%20
        (CallFunction|TriggerDigFX HitLocation)
        (CallFunction|MoveChest)
        (CallFunction|UnhideDirt)))))

(event Custom|OnMaxDug
  (bind _chestobject (Variables|Default|GetChestObject))
  (bind _chestobject_2 (Variables|Default|GetChestObject))
  (bind _itemmesh (Variables|/Game/Building/ActorBlueprints/Containers/Tiered_Chest_Athena_Parent.Tiered_Chest_Athena_Parent_C|GetItemMesh _chestobject_2))
  (bind _chestobject_3 (Variables|Default|GetChestObject))
  (bind _staticmeshcomponent (Variables|/Script/FortniteGame.BuildingSMActor|GetStaticMeshComponent _chestobject_3))
  ;[1 1 1] After Max Hits Unbind damage events and set collisions
  (Development|PrintString "MaxDug")
  (EventDispatchers|UnbindallEventsfromOnDamaged)
  (Collision|SetCollisionEnabled (Variables|Default|GetHitbox) "NoCollision")
  (if (Networking|IsServer)
    (Variables|/Game/Athena/Items/ForagedItems/BuriedChest/B_Athena_BuriedChest_Chest_Parent.B_Athena_BuriedChest_Chest_Parent_C|SetDugUp true _chestobject)
    (Variables|/Script/Engine.Actor|SetbCanBeDamaged true _chestobject)
    (Utilities|IsValid _chestobject_2
      (:Is%20Valid
        (Utilities|IsValid _itemmesh
          (:Is%20Valid
            (Collision|SetCollisionEnabled _itemmesh "QueryOnly"))
          (:Is%20Not%20Valid
            _))
        (Utilities|IsValid _staticmeshcomponent
          (:Is%20Valid
            (Collision|SetCollisionEnabled _staticmeshcomponent "QueryAndPhysics")))))))

(event Custom|TriggerDigFX
  (params (Location :type Vector))
  ;[1 1 1] Spawn FX at Location
  (Effects|SpawnEmitteratLocation (Variables|Default|GetDigParticle) (+ (Transformation|GetWorldLocation (Variables|Default|GetDirt)) "0, 0,55.000000") (Transformation|GetWorldRotation (Variables|Default|GetDirt)))
  (if (>= (Variables|Default|GetNumTimesHit) (Variables|Default|GetMaxTimesHit))
    (Effects|SpawnEmitteratLocation (Variables|Default|GetFinalHitParticle) (Transformation|GetWorldLocation (Variables|Default|GetDirt)) (Transformation|GetWorldRotation (Variables|Default|GetDirt)))))

(event Custom|ClientMoveChestLoc
  (params (ChestLoc :type Vector))
  (bind _chestobject (Variables|Default|GetChestObject))
  ;[1 1 1] Client replicated movements
  (Utilities|IsValid _chestobject
    (:Is%20Valid
      (Transformation|SetActorLocation _chestobject ChestLoc false false))))

(event Custom|MoveChest
  (bind (_returnvalue_roll _returnvalue_pitch _returnvalue_yaw) (Math|Rotator|BreakRotator (Math|Rotator|MakeRotfromZX (Variables|Default|GetChestRotNormal) (Transformation|GetActorLocation (Variables|Default|GetChestObject)))))
  (bind (_returnvalue_roll_2 _returnvalue_pitch_2 _returnvalue_yaw_2) (Math|Rotator|BreakRotator (Math|Rotator|MakeRotfromZX (Variables|Default|GetChestRotNormal) (Transformation|GetActorLocation))))
  (bind (_output_roll _output_pitch) (Math|Rotator|BreakRotator (Utilities|Array|Get(aref) (Variables|Default|GetChestRotPattern) (- (Variables|Default|GetNumTimesHit) 1))))
  (bind _numtimeshit (Variables|Default|GetNumTimesHit))
  ;[1 1 1] Handle Chest Damage
  (Development|PrintString "Dig" true true "(R=1.000000,G=0.000000,B=0.964451,A=1.000000)")
  ;[1 1 1] Set Times Dug on the chest
  (Variables|/Game/Athena/Items/ForagedItems/BuriedChest/B_Athena_BuriedChest_Chest_Parent.B_Athena_BuriedChest_Chest_Parent_C|SetTimesDug (Variables|Default|GetNumTimesHit) (Variables|Default|GetChestObject))
  ;[1 1 1] Move up and rotate on hit
  (bind _output_get (Variables|/Game/Athena/Items/ForagedItems/BuriedChest/B_Athena_BuriedChest_Chest_Parent.B_Athena_BuriedChest_Chest_Parent_C|SetChestCurrentLoc (+ (Variables|Default|GetChestInitialLoc) (* (Variables|Default|GetChestRaisePerHit) _numtimeshit)) (Variables|Default|GetChestObject)))
  (CallFunction|ClientMoveChestLoc _output_get)
  (if (>= (Variables|Default|GetNumTimesHit) (Variables|Default|GetMaxTimesHit))
    ;;[1 1 1] If it is the last hit, set max dug and set the alternate rotation
    (bind _output_get_2 (Variables|/Game/Athena/Items/ForagedItems/BuriedChest/B_Athena_BuriedChest_Chest_Parent.B_Athena_BuriedChest_Chest_Parent_C|SetChestCurrentRot (Math|Rotator|MakeRotator _returnvalue_roll _returnvalue_pitch _returnvalue_yaw) (Variables|Default|GetChestObject)))
    (CallFunction|ClientMoveChestRot _output_get_2)
    (Variables|Default|SetMaxDug true)
    (else
      (bind _output_get_3 (Variables|/Game/Athena/Items/ForagedItems/BuriedChest/B_Athena_BuriedChest_Chest_Parent.B_Athena_BuriedChest_Chest_Parent_C|SetChestCurrentRot (Math|Rotator|MakeRotator (+ _output_roll _returnvalue_roll_2) (+ _output_pitch _returnvalue_pitch_2) _returnvalue_yaw_2) (Variables|Default|GetChestObject)))
      (CallFunction|ClientMoveChestRot _output_get_3)
      (if (not (Variables|Default|GetBeenDug))
        (CallFunction|OnBeenDug)))))

(event Custom|ClientMoveChestRot
  (params (ChestRot :type Rotator))
  (bind _chestobject (Variables|Default|GetChestObject))
  ;[1 1 1] Client replicated movements
  (Utilities|IsValid _chestobject
    (:Is%20Valid
      (Transformation|SetActorRotation _chestobject ChestRot false))))

(event Custom|UnhideDirt
  (bind (_returnvalue_roll _returnvalue_pitch _returnvalue_yaw) (Math|Rotator|BreakRotator (Math|Rotator|MakeRotfromZX (Variables|Default|GetChestRotNormal) (Transformation|GetActorLocation))))
  ;[1 1 1] Unhide dirt
  (if (not (>= (Variables|Default|GetNumTimesHit) 5))
    (Variables|Default|SetDirtTransform (Math|Transform|MakeTransform (+ (Variables|/Script/Engine.SceneComponent|GetRelativeLocation (Variables|Default|GetDirt)) (Variables|Default|GetDirtMoveOnHit)) (Math|Rotator|MakeRotator _returnvalue_roll _returnvalue_pitch _returnvalue_yaw) (+ (Variables|/Script/Engine.SceneComponent|GetRelativeScale3D (Variables|Default|GetDirt)) (Variables|Default|GetDirtScaleOnHit))))))

(fn ChestObjectSet
  (Variables|Default|SetChestActor (Variables|Default|GetChestObject)))

(fn FindPlacementNormal
  (bind _outhit (Collision|LineTraceByChannel (+ (Transformation|GetActorLocation) (Variables|Default|GetTraceForChestOffset)) (- (Transformation|GetActorLocation) (Variables|Default|GetTraceForChestOffset)) "TraceTypeQuery10"))
  (Variables|Default|SetChestRotNormal (.Normal (Collision|BreakHitResult _outhit))))

(fn OnRep_ChestCurrentPos
  (Variables|/Game/Athena/Items/ForagedItems/BuriedChest/B_Athena_BuriedChest_Chest_Parent.B_Athena_BuriedChest_Chest_Parent_C|SetChestCurrentLoc (Variables|Default|GetChestCurrentPos) (Variables|Default|GetChestObject)))

(fn OnRep_ChestCurrentRot
  (Variables|/Game/Athena/Items/ForagedItems/BuriedChest/B_Athena_BuriedChest_Chest_Parent.B_Athena_BuriedChest_Chest_Parent_C|SetChestCurrentRot (Variables|Default|GetChestCurrentRot) (Variables|Default|GetChestObject)))

(fn OnRep_ChestObject
  (Utilities|IsValid (Variables|Default|GetChestObject)
    (:Is%20Valid
      (CallFunction|ChestObjectSet))))

(fn OnRep_ChestObjectPure
  (Utilities|IsValid (Variables|Default|GetChestObjectPure)
    (:Is%20Valid
      (Variables|Default|SetChestActorPure (Variables|Default|GetChestObjectPure))
      (return))
    (:Is%20Not%20Valid
      (return))))

(fn OnRep_DirtTransform
  (bind _dirttransform (Variables|Default|GetDirtTransform))
  (bind _dirt (Variables|Default|GetDirt))
  (Transformation|SetWorldRotation _dirt (.rotation _dirttransform) false false) ; This one being world not relative is intentional
  (Rendering|SetVisibility _dirt true false)
  (Transformation|SetRelativeLocation _dirt (.location _dirttransform) false false)
  (Transformation|SetRelativeScale3D _dirt (.scale _dirttransform)))

(fn OnRep_MaxDug
  (if (Variables|Default|GetMaxDug)
    (CallFunction|OnMaxDug)))

(fn ShouldDie
  (params (Damage :type Float%20%28single-precision%29) (EventInstigator :type Controller%20Object%20Reference) (DamageCauser :type Actor%20Object%20Reference) (bIsTest :type Boolean) (EffectContext :type Gameplay%20Effect%20Context%20Handle%20Structure%20%28by%20ref%29))
  (return false))

(fn ConstructionScript)
